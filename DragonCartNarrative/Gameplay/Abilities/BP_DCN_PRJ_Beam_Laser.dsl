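; Laser beam projectile for an ability that explodes where it strikes an actor other than its owner.
; On stop, a blocking hit stores the hit actor, prints its name for debugging and triggers the explosion at the impact point plus an offset.
; The server stop event repeats this logic but reads the hit from the other stop event's parameter.
; Once exploded, it hides itself and destroys itself next tick.
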(event EventBeginPlay
  (|CallParentFunction))

(event EventOnStoppedOnServer
  (if (and (Collision|BreakHitResult (stale EventOnStop Hit)) (!= (Actor|GetOwner) (.HitActor (Collision|BreakHitResult (stale EventOnStop Hit)))))
    (Variables|Default|SetHit%20Actor (.HitActor (Collision|BreakHitResult (stale EventOnStop Hit))))
    (Development|PrintString (Utilities|GetDisplayName (.HitActor (Collision|BreakHitResult (stale EventOnStop Hit)))))
    (bind _output_get (Variables|Default|SetExplosionLocation (+ (.ImpactPoint (Collision|BreakHitResult (stale EventOnStop Hit))) (Variables|Default|GetExplosionOffset))))
    (Delivery|DoExplosionatLocation _output_get)))

(event EventOnStop
  (params (Hit :type Hit%20Result%20Structure%20%28by%20ref%29))
  (bind (_bblockinghit _binitialoverlap _time _distance _location _impactpoint _normal _impactnormal _physmat _hitactor) (Collision|BreakHitResult Hit))
  (if (and _bblockinghit (!= (Actor|GetOwner) _hitactor))
    (Variables|Default|SetHit%20Actor _hitactor)
    (Development|PrintString (Utilities|GetDisplayName _hitactor))
    (bind _output_get (Variables|Default|SetExplosionLocation (+ _impactpoint (Variables|Default|GetExplosionOffset))))
    (Delivery|DoExplosionatLocation _output_get)))

(event EventOnExploded
  (params (HitActors :type Array%20of%20Actor%20Object%20References) (HitResults :type Array%20of%20Hit%20Result%20Structures))
  (Rendering|SetActorHiddenInGame self true)
  (Utilities|FlowControl|DelayUntilNextTick)
  (Actor|DestroyActor))

(fn ConstructionScript)
