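; Gameplay ability that beam-teleports a player pawn to the target rift of the BP_DCN_ShardTeleport that caused the granting effect.
; Activation hides and stasis-locks the pawn, applies tag, cue and camera mode, disables input, then teleports after level streaming.
; TeleportCompleted restores camera, input and visibility, removes the teleport effect and ends the ability.
; A fail-safe timer also forces EndAbility, whose authority branch flashes the rift and broadcasts a custom verb.
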(event Ability|EventActivateAbility
  (bind _teleport_destination (Variables|Default|GetTeleport%20Destination))
  (bind _returnvalue (Transformation|GetActorRotation _teleport_destination))
  (bind _owningpawn (Variables|Default|GetOwningPawn))
  (bind _owningpawn_2 (Variables|Default|GetOwningPawn))
  (Utilities|FlowControl|Sequence
    (:then_0
      (Variables|Default|SetOwningPawn (Utilities|Casting|CastToFortPlayerPawn (Ability|GetAvatarActorfromActorInfo)))
      (Networking|CanExecuteCosmeticEvents
        (:False
          (Stasis|SetStasisMode (Variables|Default|GetOwningPawn) "NoMovementOrFalling"))))
    (:then_1
      (bind _asbp_dcn_shard_teleport (Utilities|Casting|CastToBP_DCN_ShardTeleport (Ability|EffectContext|GetEffectCauser (Ability|GetGrantedbyEffectContext)))
        (:then
          (Ability|HasAuthority
            (:True
              (Utilities|IsValid (Variables|/DragonCartNarrative/Gameplay/BP_DCN_ShardTeleport.BP_DCN_ShardTeleport_C|GetTargetRift _asbp_dcn_shard_teleport)
                (:Is%20Valid
                  (Variables|Default|SetTeleport%20Destination (Variables|/DragonCartNarrative/Gameplay/BP_DCN_ShardTeleport.BP_DCN_ShardTeleport_C|GetTargetRift _asbp_dcn_shard_teleport)))
                (:Is%20Not%20Valid
                  (Development|PrintString "Warning: Teleport Destination is invalid!")
                  (Ability|EndAbility))))
            (:False
              _))
          (Ability|GameplayEffect|AddLooseGameplayTags (Variables|Default|GetOwningPawn) (Variables|Default|GetTeleportTag) true)
          (Ability|AddGameplayCueToOwner (Variables|Default|GetTeleportCueTag))
          (Rendering|SetActorHiddenInGame _owningpawn true)
          (Camera|PushCameraMode "/Game/Blueprints/Camera/Athena/Athena_PlayerCameraMode_HidingProp_Teleport.Athena_PlayerCameraMode_HidingProp_Teleport_C")
          (Pawn|Components|CharacterMovement|SetMovementMode (Variables|/Script/Engine.Character|GetCharacterMovement _owningpawn) "MOVE_Falling" 0)
          (Utilities|FlowControl|Delay 0.25)
          (Input|DisableInput (Variables|Default|GetOwningPawn) (Fortnite|GetFortPlayerControllerfromActor (Variables|Default|GetOwningPawn)))
          (Class|FortAsyncActionTeleportandWaitforLevelStreaming|CreateTeleportandWaitforLevelStreaming _owningpawn_2 (Transformation|GetActorLocation _teleport_destination) _returnvalue
            (:OnComplete
              (Transform|Actor|TeleportPlayerPawn _owningpawn_2 (Transformation|GetActorLocation _owningpawn_2) _returnvalue true true)
              (Utilities|FlowControl|Delay 1.0)
              (CallFunction|TeleportCompleted))))
        (:CastFailed
          (Development|PrintString "Invalid Causer" true true "(R=0.000000,G=0.660000,B=1.000000,A=1.000000)" 2e+01)
          (Ability|EndAbility))))
    (:then_2
      (bind _returnvalue_2 (Utilities|Time|SetTimerbyEvent (delegate Custom|ForceEndAbility) 1e+01))
      (Variables|Default|SetTH_FailSafeRemoval _returnvalue_2))))

(event Ability|EventOnEndAbility
  (params (bWasCancelled :type Boolean))
  (bind _returnvalue (Ability|GetAvatarActorfromActorInfo))
  (bind _asfort_player_pawn (Utilities|Casting|CastToFortPlayerPawn _returnvalue))
  (Ability|GameplayEffect|RemoveLooseGameplayTags (Variables|Default|GetOwningPawn) (Variables|Default|GetTeleportTag) true)
  (Rendering|SetActorHiddenInGame _asfort_player_pawn false)
  (Stasis|SetStasisMode _asfort_player_pawn "None")
  (Ability|HasAuthority
    (:True
      (Class|BPDCNTeleportLocation|RiftFlash (Variables|Default|GetTeleport%20Destination))
      (Utilities|Time|ClearandInvalidateTimerbyHandle (Variables|Default|GetTH_FailSafeRemoval))
      (CustomVerb|BroadcastCustomVerbMessage (CustomVerb|MakeCustomVerbMessage (Variables|Default|GetOwningPawn) (Variables|Default|GetShardTeleportCustomVerb))))))

(event Custom|TeleportCompleted
  (Camera|PopCameraMode)
  (Rendering|SetActorHiddenInGame (Variables|Default|GetOwningPawn) false)
  (Utilities|FlowControl|Delay 0.1)
  (Input|EnableInput (Variables|Default|GetOwningPawn) (Fortnite|GetFortPlayerControllerfromActor (Variables|Default|GetOwningPawn)))
  (Stasis|SetStasisMode (Variables|Default|GetOwningPawn) "None")
  (Ability|RemoveGameplayCueFromOwner (Variables|Default|GetTeleportCueTag))
  (GameplayEffects|RemoveActiveGameplayEffectbySourceEffect (Ability|GetFortAbilitySystemComponent (Variables|Default|GetOwningPawn)) (Variables|Default|GetTeleportEffect) 0 -1)
  (Ability|EndAbility))

(event Custom|ForceEndAbility
  (Ability|EndAbility))

(fn OnRep_PlayCue)
